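# Physics asset providing cloth collision for the Black Monday Houston cape back accessory.
# It defines a single body on the backpack_root bone, built from two capsules of different sizes, both rotated 90 degrees.
# It has no constraints, so it only supplies collision volumes, presumably for the cape's cloth simulation to rest against.
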PhysicsAsset /BRCosmetics/Accessories/FORT_Capes/M_MED_BlackMonday_Houston_Cape/Meshes/M_MED_BlackMonday_Houston_Cape_ClothCollision
bodies {backpack_root [
  {shapeType=Capsule, length=8.7159, radius=6.1934, center=<0 -0.7239 1.8108>, rotation=<rot 90 0 0>}
  {shapeType=Capsule, length=37.0262, radius=13.9803, shapeName=backpack_root_capsule, center=<-0.2331 -12.2029 9.9038>, rotation=<rot 90 0 0>}
 ]}
constraints []
