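# Anim Blueprint for the BlackMonday Houston cape, posed from the attached parent mesh with clavicle and cape-root offsets.
# With cloth enabled it outputs that pose, adding trail controllers at low LOD; otherwise it runs RigidBody simulation with arm-driven offsets while skydiving or parachuting.
# The EventGraph tunes Houston and Kansas-named clothing by movement state and resets cloth around ziplines.
# The DBNO, skydive and grounded cached-pose chains never reach Output Pose.
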AnimBlueprint /BRCosmetics/Accessories/FORT_Capes/M_MED_BlackMonday_Houston_Cape/Meshes/M_MED_BlackMonday_Houston_Pack_AnimBP
target_skeleton=/Game/Accessories/FORT_Capes/M_MED_BlackMonday_Houston_Cape/Meshes/M_MED_BlackMonday_Houston_Pack_Skeleton
graphs [
 AnimGraph {nodes [
   Blend Poses by bool(
    BlendPose_0=Use cached pose 'dbno_post_sim'()
    BlendPose_1=Blend Poses by bool(BlendPose_0=Use cached pose 'skydive_post_sim'(), BlendPose_1=Use cached pose 'grounded_motion_cache'(), bActiveValue=Get bIsSkyDiving()) {blendType=Linear}
    bActiveValue=Get bIsDBNO()
   ) {blendType=Linear}
   n1 = RigidBody!(
    ComponentPose=Local To Component(LocalPose=Use cached pose 'Start_Cache'())
    bAlphaBoolEnabled:bool
    OverrideWorldGravity:Vector=Get GravityOverride()
    ExternalForce
   ) -> (Pose) {
    alphaInputType=Bool
    bClampLinearTranslationLimitToRefPose=true
    bOverrideWorldGravity=true
    overridePhysicsAsset=&/M_MED_BlackMonday_Houston_Pack_Physics
    simulationSpace=BaseBoneSpace
    alphaBoolBlend {blendOption=Linear}
    baseBoneRef {boneName=pelvis}
    componentLinearAccScale=<0.4 0.4 0.5>
    componentLinearVelScale=<0.4 0.4 0.5>
    overrideWorldGravity {z=-980}
   }
   n2 = Get upperarm_Alpha() -> (upperarm_Alpha:double)
   n3 = "Transform (Modify) Bone - Bone: dyn_cape_root"(ComponentPose=@n1, Alpha=@n2, Translation="0,-10,5", Rotation="0,0,20", Scale="1,1,1") {
    rotationMode=BMM_Additive
    rotationSpace=BCS_BoneSpace
    translationMode=BMM_Additive
    alphaBoolBlend {blendOption=Linear}
    boneToModify {boneName=dyn_cape_root}
    rotation {roll=20}
    translation {y=-10, z=5}
   }
   n4 = "Transform (Modify) Bone - Bone: dyn_cape01"(ComponentPose=@n3, Alpha=@n2, Translation="0,-8,-8.5", Rotation="0,0,0", Scale="1,1,1") {
    rotationMode=BMM_Additive
    rotationSpace=BCS_BoneSpace
    translationMode=BMM_Additive
    alphaBoolBlend {blendOption=Linear}
    boneToModify {boneName=dyn_cape01}
    translation {y=-8, z=-8.5}
   }
   Output Pose(Result=Blend Poses by bool(BlendPose_0=Use cached pose 'trailCache'(), BlendPose_1=Component To Local(ComponentPose=@n4), bActiveValue=Get bClothEnabled()) {blendType=Linear}) {layerGroup=None}
   n5 = "Transform (Modify) Bone - Bone: clavicle_l"!(
    ComponentPose=Local To Component(LocalPose=Copy Pose From Mesh() {bUseAttachedParent=true})
    Alpha=1
    Translation="0,0,0"
    Rotation="-0.000,-3,-0.000"
   ) -> (Pose) {
    rotationMode=BMM_Additive
    rotationSpace=BCS_BoneSpace
    translationMode=BMM_Additive
    translationSpace=BCS_BoneSpace
    alphaBoolBlend {blendOption=Linear}
    alphaScaleBiasClamp {bInterpResult=true, interpSpeedDecreasing=5, interpSpeedIncreasing=7}
    boneToModify {boneName=clavicle_l}
    rotation {yaw=-3}
   }
   n6 = "Transform (Modify) Bone - Bone: clavicle_r"!(ComponentPose=@n5, Alpha=1, Translation="0,0,0", Rotation="-0.000,-3,-0.000") -> (Pose) {
    rotationMode=BMM_Additive
    rotationSpace=BCS_BoneSpace
    translationMode=BMM_Additive
    translationSpace=BCS_BoneSpace
    alphaBoolBlend {blendOption=Linear}
    alphaScaleBiasClamp {bInterpResult=true, interpSpeedDecreasing=5, interpSpeedIncreasing=7}
    boneToModify {boneName=clavicle_r}
    rotation {yaw=-3}
   }
   n7 = "Transform (Modify) Bone - Bone: clavicle_l"(ComponentPose=@n6, Alpha=1, Translation="0,0,0", Rotation="0,-5,0") {
    rotationMode=BMM_Additive
    rotationSpace=BCS_BoneSpace
    translationMode=BMM_Additive
    translationSpace=BCS_BoneSpace
    alphaBoolBlend {blendOption=Linear}
    alphaScaleBiasClamp {bInterpResult=true, interpSpeedDecreasing=5, interpSpeedIncreasing=7}
    boneToModify {boneName=clavicle_l}
    rotation {yaw=-5}
   }
   n8 = "Transform (Modify) Bone - Bone: clavicle_r"(ComponentPose=@n7, Alpha=1, Translation="0,0,0", Rotation="-0.000,-4.5,-0.000") {
    rotationMode=BMM_Additive
    rotationSpace=BCS_BoneSpace
    translationMode=BMM_Additive
    translationSpace=BCS_BoneSpace
    alphaBoolBlend {blendOption=Linear}
    alphaScaleBiasClamp {bInterpResult=true, interpSpeedDecreasing=5, interpSpeedIncreasing=7}
    boneToModify {boneName=clavicle_r}
    rotation {yaw=-4.5}
   }
   n9 = "Transform (Modify) Bone - Bone: dyn_cape_root"!(ComponentPose=@n8, Alpha=1, Translation="0,-7,0", Rotation="0,0,0") -> (Pose) {
    rotationMode=BMM_Additive
    rotationSpace=BCS_BoneSpace
    translationMode=BMM_Additive
    translationSpace=BCS_BoneSpace
    alphaBoolBlend {blendOption=Linear}
    alphaScaleBiasClamp {bInterpResult=true, interpSpeedDecreasing=5, interpSpeedIncreasing=7}
    boneToModify {boneName=dyn_cape_root}
    translation {y=-7}
   }
   Save cached pose 'Start_Cache'(Pose=Component To Local(ComponentPose=@n9)) {cacheName=Start_Cache}
   n10 = "Transform (Modify) Bone - Bone: dyn_cape01"!(
    ComponentPose=Local To Component(LocalPose=Use cached pose 'Start_Cache'())
    bAlphaBoolEnabled:bool=Get trailControllerConditions() -> (trailControllerConditions:bool)
    Translation="0,0,0"
    Rotation=Get SkydivingAlpha() -> (SkydivingAlpha:Rotator)
    Scale="1,1,1"
   ) -> (Pose) {
    alphaInputType=Bool
    rotationMode=BMM_Additive
    rotationSpace=BCS_BoneSpace
    alphaBoolBlend {blendOption=Linear}
    boneToModify {boneName=dyn_cape01}
   }
   n11 = Get BwdMovement() -> (BwdMovement:double)
   n12 = Trail controller - Bone: dyn_cape05(ComponentPose=@n10, Alpha=@n11) {
    bLimitRotation=true
    bShowBaseMotion=false
    bShowLimit=false
    chainLength=5
    debugLifeTime=0.1
    maxDeltaTime=0.333
    alphaScaleBiasClamp {bInterpResult=true}
    planarLimits [{drivingBone {boneName=dyn_cape_root}, planeTransform {location=<0 -3 0>, rotation=<rot 0 0 -90>, scale=<1 1 1>}}]
    rotationLimits [
     {limitMax=<180 180 0>, limitMin=<-180 -180 -30>}
     {limitMax=<180 180 0>, limitMin=<-180 -180 -30>}
     {limitMax=<180 180 0>, limitMin=<-180 -180 -20>}
     {limitMax=<180 180 0>, limitMin=<-180 -180 -20>}
     {limitMax=<180 180 0>, limitMin=<-180 -180 -20>}
    ]
    rotationOffsets [<0 0 0>, <0 0 0>, <0 0 0>, <0 0 0>, <0 0 0>]
    trailBone {boneName=dyn_cape05}
    trailRelaxationSpeed {editorCurveData {keys [<at -0.0134 19.3841 RCIM_Linear>,<at 0.5079 12.8986 RCIM_Linear>,<at 0.9955 9.6044 RCIM_Linear>]}}
   }
   n13 = Trail controller - Bone: dyn_cape05_l(ComponentPose=@n12, Alpha=@n11) {
    bLimitRotation=true
    bShowBaseMotion=false
    bShowLimit=false
    chainLength=5
    debugLifeTime=0.1
    maxDeltaTime=0.0333
    alphaScaleBiasClamp {bInterpResult=true}
    planarLimits [{drivingBone {boneName=dyn_cape_root}, planeTransform {location=<0 0 0>, rotation=<rot 0 0 -90>, scale=<1 1 1>}}]
    rotationLimits [
     {limitMax=<180 180 0>, limitMin=<-180 -180 -20>}
     {limitMax=<180 180 0>, limitMin=<-180 -180 -20>}
     {limitMax=<180 90 0>, limitMin=<-180 -90 -20>}
     {limitMax=<180 180 0>, limitMin=<-180 -180 -20>}
     {limitMax=<180 180 0>, limitMin=<-180 -180 -20>}
    ]
    rotationOffsets [<0 0 0>, <0 0 0>, <0 0 0>, <0 0 0>, <0 0 0>]
    trailBone {boneName=dyn_cape05_l}
    trailRelaxationSpeed {editorCurveData {keys [<at 0 15 RCIM_Linear>,<at 0.5289 13.1838 RCIM_Linear>,<at 1 6 RCIM_Linear>]}}
   }
   n14 = Trail controller - Bone: dyn_cape05_r(ComponentPose=@n13, Alpha=@n11) {
    bLimitRotation=true
    bShowBaseMotion=false
    bShowLimit=false
    chainLength=5
    debugLifeTime=0.1
    maxDeltaTime=0.0333
    alphaScaleBiasClamp {bInterpResult=true}
    planarLimits [{drivingBone {boneName=dyn_cape_root}, planeTransform {location=<0 0 0>, rotation=<rot 0 0 -90>, scale=<1 1 1>}}]
    rotationLimits [
     {limitMax=<180 180 0>, limitMin=<-180 -180 -20>}
     {limitMax=<180 180 0>, limitMin=<-180 -180 -20>}
     {limitMax=<180 90 0>, limitMin=<-180 -90 -20>}
     {limitMax=<180 180 0>, limitMin=<-180 -180 -20>}
     {limitMax=<180 180 0>, limitMin=<-180 -180 -20>}
    ]
    rotationOffsets [<0 0 0>, <0 0 0>, <0 0 0>, <0 0 0>, <0 0 0>]
    trailBone {boneName=dyn_cape05_r}
    trailRelaxationSpeed {editorCurveData {keys [<at 0 15 RCIM_Linear>,<at 0.5289 13.1838 RCIM_Linear>,<at 1 6 RCIM_Linear>]}}
   }
   Save cached pose 'trailCache'(Pose=Blend Poses by bool(
     BlendPose_0=Component To Local(ComponentPose=@n14)
     BlendPose_1=Use cached pose 'Start_Cache'()
     bActiveValue=Get trailControllerConditions()
    ) {blendType=Linear}) {cacheName=trailCache}
   n15 = Get DBNOTrailAlpha() -> (DBNOTrailAlpha:float)
   n16 = Trail controller - Bone: cape_mid_07_l!(
    ComponentPose=Local To Component(LocalPose=Use cached pose 'dbno_cache'())
    Alpha=@n15
    RelaxationSpeedScale:float=2
    FakeVelocity:Vector="0,0,0"
   ) -> (Pose) {
    bActorSpaceFakeVel=true
    bShowTrailLocation=true
    bUsePlanarLimit=true
    chainLength=6
    debugLifeTime=0.1
    maxDeltaTime=0.0333
    relaxationSpeedScale=1.2
    alphaScaleBiasClamp {bInterpResult=true, interpSpeedDecreasing=25, interpSpeedIncreasing=25}
    baseJoint {boneName=root}
    rotationLimits [
     {limitMax=<180 180 180>}
     {limitMax=<180 180 180>}
     {limitMax=<180 180 180>}
     {limitMax=<180 180 180>}
     {limitMax=<180 180 180>}
     {limitMax=<180 180 180>}
    ]
    rotationOffsets [<0 0 0>, <0 0 0>, <0 0 0>, <0 0 0>, <0 0 0>, <0 0 0>]
    trailBone {boneName=cape_mid_07_l}
    trailRelaxationSpeed {editorCurveData {keys [<at 0 15 RCIM_Linear>,<at 1 6 RCIM_Linear>]}}
   }
   n17 = Trail controller - Bone: cape_mid_07_r!(ComponentPose=@n16, Alpha=@n15, RelaxationSpeedScale:float=2, FakeVelocity:Vector="0,0,0") -> (Pose) {
    bActorSpaceFakeVel=true
    bShowTrailLocation=true
    bUsePlanarLimit=true
    chainLength=6
    debugLifeTime=0.1
    maxDeltaTime=0.0333
    relaxationSpeedScale=1.2
    alphaScaleBiasClamp {bInterpResult=true, interpSpeedDecreasing=25, interpSpeedIncreasing=25}
    baseJoint {boneName=root}
    rotationLimits [
     {limitMax=<180 180 180>}
     {limitMax=<180 180 180>}
     {limitMax=<180 180 180>}
     {limitMax=<180 180 180>}
     {limitMax=<180 180 180>}
     {limitMax=<180 180 180>}
    ]
    rotationOffsets [<0 0 0>, <0 0 0>, <0 0 0>, <0 0 0>, <0 0 0>, <0 0 0>]
    trailBone {boneName=cape_mid_07_r}
    trailRelaxationSpeed {editorCurveData {keys [<at 0 15 RCIM_Linear>,<at 1 6 RCIM_Linear>]}}
   }
   n18 = Trail controller - Bone: cape_08_l!(ComponentPose=@n17, Alpha=@n15, RelaxationSpeedScale:float=2, FakeVelocity:Vector="0,0,0") -> (Pose) {
    bActorSpaceFakeVel=true
    bShowTrailLocation=true
    bUsePlanarLimit=true
    chainLength=6
    debugLifeTime=0.1
    maxDeltaTime=0.0333
    relaxationSpeedScale=1.2
    alphaScaleBiasClamp {bInterpResult=true, interpSpeedDecreasing=25, interpSpeedIncreasing=25}
    baseJoint {boneName=root}
    rotationLimits [
     {limitMax=<180 180 180>}
     {limitMax=<180 180 180>}
     {limitMax=<180 180 180>}
     {limitMax=<180 180 180>}
     {limitMax=<180 180 180>}
     {limitMax=<180 180 180>}
    ]
    rotationOffsets [<0 0 0>, <0 0 0>, <0 0 0>, <0 0 0>, <0 0 0>, <0 0 0>]
    trailBone {boneName=cape_08_l}
    trailRelaxationSpeed {editorCurveData {keys [<at 0 15 RCIM_Linear>,<at 1 6 RCIM_Linear>]}}
   }
   n19 = Trail controller - Bone: cape_08_r!(ComponentPose=@n18, Alpha=@n15, RelaxationSpeedScale:float=2, FakeVelocity:Vector="0,0,0") -> (Pose) {
    bActorSpaceFakeVel=true
    bShowTrailLocation=true
    bUsePlanarLimit=true
    chainLength=6
    debugLifeTime=0.1
    maxDeltaTime=0.0333
    relaxationSpeedScale=1.2
    alphaScaleBiasClamp {bInterpResult=true, interpSpeedDecreasing=25, interpSpeedIncreasing=25}
    baseJoint {boneName=root}
    rotationLimits [
     {limitMax=<180 180 180>}
     {limitMax=<180 180 180>}
     {limitMax=<180 180 180>}
     {limitMax=<180 180 180>}
     {limitMax=<180 180 180>}
     {limitMax=<180 180 180>}
    ]
    rotationOffsets [<0 0 0>, <0 0 0>, <0 0 0>, <0 0 0>, <0 0 0>, <0 0 0>]
    trailBone {boneName=cape_08_r}
    trailRelaxationSpeed {editorCurveData {keys [<at 0 15 RCIM_Linear>,<at 1 6 RCIM_Linear>]}}
   }
   Save cached pose 'dbno_post_sim'(Pose=Component To Local(ComponentPose=@n19)) {cacheName=dbno_post_sim}
   Save cached pose 'dbno_cache'() {cacheName=dbno_cache}
   Save cached pose 'copypose_cache'(Pose=Apply Additive(Alpha=Get DBNOAdditiveAlpha() -> (DBNOAdditiveAlpha:float)) {alphaBoolBlend {blendOption=Linear}}) {cacheName=copypose_cache}
   Save cached pose 'grounded_pre_sim_cache'() {cacheName=grounded_pre_sim_cache}
   Save cached pose 'skydive_pre_sim_cache'() {cacheName=skydive_pre_sim_cache}
   n20 = Get Is not In Front End() -> (Is not In Front End:double)
   n21 = Get SkyDiveTrailAlpha() -> (SkyDiveTrailAlpha:float)
   n22 = BlackMonday_Houston_Poses_PoseAsset : Run!(PoseWeight:float=1) -> (Pose:Pose Link Structure) {
    is=title AnimGraphNode_PoseByName
    poseAsset=/BRCosmetics/Accessories/FORT_Capes/M_MED_BlackMonday_Houston_Cape/Animation/BlackMonday_Houston_Poses_PoseAsset
    poseName=Run
   }
   n23 = BlackMonday_Houston_Poses_PoseAsset : Sprint!(PoseWeight:float=0.2) -> (Pose:Pose Link Structure) {
    is=title AnimGraphNode_PoseByName
    poseAsset=&BlackMonday_Houston_Poses_PoseAsset
    poseName=Sprint
   }
   n24 = Get SprintSpeedAlpha() -> (SprintSpeedAlpha:float)
   n25 = Two Way Blend(
    A=Apply Additive(Base=Use cached pose 'grounded_pre_sim_cache'(), Additive=@n23, Alpha=@n24) {alphaBoolBlend {blendOption=Linear}, alphaScaleBiasClamp {bInterpResult=true, interpSpeedDecreasing=5}}
    B=Apply Additive(Base=Use cached pose 'grounded_pre_sim_cache'(), Additive=@n22, Alpha=Get RunSpeedAlpha() -> (RunSpeedAlpha:float)) {alphaBoolBlend {blendOption=Linear}, alphaScaleBiasClamp {bInterpResult=true, interpSpeedDecreasing=5}}
    Alpha=@n24
   ) {blendNode {
     alphaBoolBlend {blendOption=Linear}
     alphaScaleBiasClamp {bInterpResult=true, bMapRange=true, interpSpeedDecreasing=5, outRange {max=0, min=1}}
    }}
   n26 = Use cached pose 'grounded_pre_sim_cache'()
   n27 = Get bIsCrouching()
   n28 = Get ShoulderCorrectionAlpha() -> (ShoulderCorrectionAlpha:float)
   n29 = "Transform (Modify) Bone - Bone: backpack_root"(
    ComponentPose=Local To Component(LocalPose=Blend Poses by bool(BlendPose_0=@n26, BlendPose_1=@n25, BlendTime_0=0.3, BlendTime_1=0.3, bActiveValue=@n27) {blendType=Linear})
    Alpha=@n28
    Translation="0,0,0"
    Rotation="0.000,0.000,17.86"
    Scale="1,1,1"
   ) {
    rotationMode=BMM_Additive
    rotationSpace=BCS_BoneSpace
    translationMode=BMM_Additive
    translationSpace=BCS_BoneSpace
    alphaBoolBlend {blendOption=Linear}
    alphaScaleBiasClamp {bInterpResult=true}
    boneToModify {boneName=backpack_root}
    rotation {roll=17.86}
   }
   n30 = "Transform (Modify) Bone - Bone: cape_01_l"(ComponentPose=@n29, Alpha=@n28, Translation="0,0,0", Rotation="5,-20.000,0.000", Scale="1,1,1") {
    rotationMode=BMM_Additive
    rotationSpace=BCS_BoneSpace
    alphaBoolBlend {blendOption=Linear}
    alphaScaleBiasClamp {bInterpResult=true}
    boneToModify {boneName=cape_01_l}
    rotation {pitch=5, yaw=-20}
   }
   n31 = "Transform (Modify) Bone - Bone: cape_01_r"(ComponentPose=@n30, Alpha=@n28, Translation="0,0,0", Rotation="-5,-20,0", Scale="1,1,1") {
    rotationMode=BMM_Additive
    rotationSpace=BCS_BoneSpace
    alphaBoolBlend {blendOption=Linear}
    alphaScaleBiasClamp {bInterpResult=true}
    boneToModify {boneName=cape_01_r}
    rotation {pitch=-5, yaw=-20}
   }
   n32 = Trail controller - Bone: cape_mid_07_l(
    ComponentPose=@n31
    Alpha=@n21
    RelaxationSpeedScale=Get CapeTrailRelaxSpeedScale() -> (CapeTrailRelaxSpeedScale:float)
    FakeVelocity=vector * vector(A:double=@n20, B:Vector=Get CapeTrailFakeVelocityMidLeft() -> (CapeTrailFakeVelocityMidLeft:Vector)) -> (ReturnValue=~)
   ) {
    bActorSpaceFakeVel=true
    bShowTrailLocation=true
    bUsePlanarLimit=true
    chainLength=6
    debugLifeTime=0.1
    maxDeltaTime=0.0333
    relaxationSpeedScale=1.2
    alphaScaleBiasClamp {bInterpResult=true, interpSpeedDecreasing=25, interpSpeedIncreasing=25}
    baseJoint {boneName=root}
    planarLimits [
     {drivingBone {boneName=root}, planeTransform {location=<0 0 10>, rotation=<rot 0 0 0>, scale=<1 1 1>}}
     {drivingBone {boneName=spine_03}, planeTransform {location=<0 -20 0>, rotation=<rot 0 15 -90>, scale=<1 1 1>}}
    ]
    rotationLimits [
     {limitMax=<180 180 180>}
     {limitMax=<180 180 180>}
     {limitMax=<180 180 180>}
     {limitMax=<180 180 180>}
     {limitMax=<180 180 180>}
     {limitMax=<180 180 180>}
    ]
    rotationOffsets [<0 0 0>, <0 0 0>, <0 0 0>, <0 0 0>, <0 0 0>, <0 0 0>]
    trailBone {boneName=cape_mid_07_l}
    trailRelaxationSpeed {editorCurveData {keys [<at 0 15 RCIM_Linear>,<at 0.5289 13.1838 RCIM_Linear>,<at 1 6 RCIM_Linear>]}}
   }
   n33 = Trail controller - Bone: cape_mid_07_r(
    ComponentPose=@n32
    Alpha=@n21
    RelaxationSpeedScale=Get CapeTrailRelaxSpeedScale()
    FakeVelocity=vector * vector(A=@n20, B=Get CapeTrailFakeVelocityMidRight() -> (CapeTrailFakeVelocityMidRight:Vector)) -> (ReturnValue=~)
   ) {
    bActorSpaceFakeVel=true
    bShowTrailLocation=true
    bUsePlanarLimit=true
    chainLength=6
    debugLifeTime=0.1
    maxDeltaTime=0.0333
    relaxationSpeedScale=1.2
    alphaScaleBiasClamp {bInterpResult=true, interpSpeedDecreasing=25, interpSpeedIncreasing=25}
    baseJoint {boneName=root}
    planarLimits [
     {drivingBone {boneName=root}, planeTransform {location=<0 0 10>, rotation=<rot 0 0 0>, scale=<1 1 1>}}
     {drivingBone {boneName=spine_03}, planeTransform {location=<0 -20 0>, rotation=<rot 0 15 -89.9999>, scale=<1 1 1>}}
    ]
    rotationLimits [
     {limitMax=<180 180 180>}
     {limitMax=<180 180 180>}
     {limitMax=<180 180 180>}
     {limitMax=<180 180 180>}
     {limitMax=<180 180 180>}
     {limitMax=<180 180 180>}
    ]
    rotationOffsets [<0 0 0>, <0 0 0>, <0 0 0>, <0 0 0>, <0 0 0>, <0 0 0>]
    trailBone {boneName=cape_mid_07_r}
    trailRelaxationSpeed {editorCurveData {keys [<at 0 15 RCIM_Linear>,<at 0.5289 13.1838 RCIM_Linear>,<at 1 6 RCIM_Linear>]}}
   }
   n34 = Trail controller - Bone: cape_08_l(
    ComponentPose=@n33
    Alpha=@n21
    RelaxationSpeedScale=Get CapeTrailRelaxSpeedScale()
    FakeVelocity=vector * vector(A=@n20, B=Get CapeTrailFakeVelocityLowLeft() -> (CapeTrailFakeVelocityLowLeft:Vector)) -> (ReturnValue=~)
   ) {
    bActorSpaceFakeVel=true
    bShowTrailLocation=true
    bUsePlanarLimit=true
    chainLength=6
    debugLifeTime=0.1
    maxDeltaTime=0.0333
    relaxationSpeedScale=1.2
    alphaScaleBiasClamp {bInterpResult=true, interpSpeedDecreasing=25, interpSpeedIncreasing=25}
    baseJoint {boneName=root}
    planarLimits [
     {drivingBone {boneName=root}, planeTransform {location=<0 0 10>, rotation=<rot 0 0 0>, scale=<1 1 1>}}
     {drivingBone {boneName=spine_03}, planeTransform {location=<0 -20 0>, rotation=<rot 0 15 -89.9999>, scale=<1 1 1>}}
    ]
    rotationLimits [
     {limitMax=<180 180 180>}
     {limitMax=<180 180 180>}
     {limitMax=<180 180 180>}
     {limitMax=<180 180 180>}
     {limitMax=<180 180 180>}
     {limitMax=<180 180 180>}
    ]
    rotationOffsets [<0 0 0>, <0 0 0>, <0 0 0>, <0 0 0>, <0 0 0>, <0 0 0>]
    trailBone {boneName=cape_08_l}
    trailRelaxationSpeed {editorCurveData {keys [<at 0 15 RCIM_Linear>,<at 0.5289 13.1838 RCIM_Linear>,<at 1 6 RCIM_Linear>]}}
   }
   n35 = Trail controller - Bone: cape_08_r(
    ComponentPose=@n34
    Alpha=@n21
    RelaxationSpeedScale=Get CapeTrailRelaxSpeedScale()
    FakeVelocity=vector * vector(A=@n20, B=Get CapeTrailFakeVelocityLowRight() -> (CapeTrailFakeVelocityLowRight:Vector)) -> (ReturnValue=~)
   ) {
    bActorSpaceFakeVel=true
    bShowTrailLocation=true
    bUsePlanarLimit=true
    chainLength=6
    debugLifeTime=0.1
    maxDeltaTime=0.0333
    relaxationSpeedScale=1.2
    alphaScaleBiasClamp {bInterpResult=true, interpSpeedDecreasing=25, interpSpeedIncreasing=25}
    baseJoint {boneName=root}
    planarLimits [
     {drivingBone {boneName=root}, planeTransform {location=<0 0 10>, rotation=<rot 0 0 0>, scale=<1 1 1>}}
     {drivingBone {boneName=spine_03}, planeTransform {location=<0 -20 0>, rotation=<rot 0 15 -89.9999>, scale=<1 1 1>}}
    ]
    rotationLimits [
     {limitMax=<180 180 180>}
     {limitMax=<180 180 180>}
     {limitMax=<180 180 180>}
     {limitMax=<180 180 180>}
     {limitMax=<180 180 180>}
     {limitMax=<180 180 180>}
    ]
    rotationOffsets [<0 0 0>, <0 0 0>, <0 0 0>, <0 0 0>, <0 0 0>, <0 0 0>]
    trailBone {boneName=cape_08_r}
    trailRelaxationSpeed {editorCurveData {keys [<at 0 15 RCIM_Linear>,<at 0.5289 13.1838 RCIM_Linear>,<at 1 6 RCIM_Linear>]}}
   }
   Save cached pose 'grounded_motion_cache'(Pose=Component To Local(ComponentPose=@n35)) {cacheName=grounded_motion_cache}
   n36 = Get SkyDiveTrailAlpha()
   n37 = Trail controller - Bone: cape_mid_07_l(
    ComponentPose=Local To Component(LocalPose=Use cached pose 'skydive_pre_sim_cache'())
    Alpha=@n36
    RelaxationSpeedScale=3
    FakeVelocity="0,0,0"
   ) {
    bActorSpaceFakeVel=true
    bShowTrailLocation=true
    bUsePlanarLimit=true
    chainLength=6
    debugLifeTime=0.1
    maxDeltaTime=0.0333
    relaxationSpeedScale=1.2
    alphaScaleBiasClamp {bInterpResult=true, interpSpeedDecreasing=25, interpSpeedIncreasing=25}
    baseJoint {boneName=root}
    rotationLimits [
     {limitMax=<180 180 180>}
     {limitMax=<180 180 180>}
     {limitMax=<180 180 180>}
     {limitMax=<180 180 180>}
     {limitMax=<180 180 180>}
     {limitMax=<180 180 180>}
    ]
    rotationOffsets [<0 0 0>, <0 0 0>, <0 0 0>, <0 0 0>, <0 0 0>, <0 0 0>]
    trailBone {boneName=cape_mid_07_l}
    trailRelaxationSpeed {editorCurveData {keys [<at 0 15 RCIM_Linear>,<at 1 6 RCIM_Linear>]}}
   }
   n38 = Trail controller - Bone: cape_mid_07_r(ComponentPose=@n37, Alpha=@n36, RelaxationSpeedScale=3, FakeVelocity="0,0,0") {
    bActorSpaceFakeVel=true
    bShowTrailLocation=true
    bUsePlanarLimit=true
    chainLength=6
    debugLifeTime=0.1
    maxDeltaTime=0.0333
    relaxationSpeedScale=1.2
    alphaScaleBiasClamp {bInterpResult=true, interpSpeedDecreasing=25, interpSpeedIncreasing=25}
    baseJoint {boneName=root}
    rotationLimits [
     {limitMax=<180 180 180>}
     {limitMax=<180 180 180>}
     {limitMax=<180 180 180>}
     {limitMax=<180 180 180>}
     {limitMax=<180 180 180>}
     {limitMax=<180 180 180>}
    ]
    rotationOffsets [<0 0 0>, <0 0 0>, <0 0 0>, <0 0 0>, <0 0 0>, <0 0 0>]
    trailBone {boneName=cape_mid_07_r}
    trailRelaxationSpeed {editorCurveData {keys [<at 0 15 RCIM_Linear>,<at 1 6 RCIM_Linear>]}}
   }
   n39 = Trail controller - Bone: cape_08_l(ComponentPose=@n38, Alpha=@n36, RelaxationSpeedScale=3, FakeVelocity="0,0,0") {
    bActorSpaceFakeVel=true
    bShowTrailLocation=true
    bUsePlanarLimit=true
    chainLength=6
    debugLifeTime=0.1
    maxDeltaTime=0.0333
    relaxationSpeedScale=1.2
    alphaScaleBiasClamp {bInterpResult=true, interpSpeedDecreasing=25, interpSpeedIncreasing=25}
    baseJoint {boneName=root}
    rotationLimits [
     {limitMax=<180 180 180>}
     {limitMax=<180 180 180>}
     {limitMax=<180 180 180>}
     {limitMax=<180 180 180>}
     {limitMax=<180 180 180>}
     {limitMax=<180 180 180>}
    ]
    rotationOffsets [<0 0 0>, <0 0 0>, <0 0 0>, <0 0 0>, <0 0 0>, <0 0 0>]
    trailBone {boneName=cape_08_l}
    trailRelaxationSpeed {editorCurveData {keys [<at 0 15 RCIM_Linear>,<at 1 6 RCIM_Linear>]}}
   }
   n40 = Trail controller - Bone: cape_08_r(ComponentPose=@n39, Alpha=@n36, RelaxationSpeedScale=3, FakeVelocity="0,0,0") {
    bActorSpaceFakeVel=true
    bShowTrailLocation=true
    bUsePlanarLimit=true
    chainLength=6
    debugLifeTime=0.1
    maxDeltaTime=0.0333
    relaxationSpeedScale=1.2
    alphaScaleBiasClamp {bInterpResult=true, interpSpeedDecreasing=25, interpSpeedIncreasing=25}
    baseJoint {boneName=root}
    rotationLimits [
     {limitMax=<180 180 180>}
     {limitMax=<180 180 180>}
     {limitMax=<180 180 180>}
     {limitMax=<180 180 180>}
     {limitMax=<180 180 180>}
     {limitMax=<180 180 180>}
    ]
    rotationOffsets [<0 0 0>, <0 0 0>, <0 0 0>, <0 0 0>, <0 0 0>, <0 0 0>]
    trailBone {boneName=cape_08_r}
    trailRelaxationSpeed {editorCurveData {keys [<at 0 15 RCIM_Linear>,<at 1 6 RCIM_Linear>]}}
   }
   Save cached pose 'skydive_post_sim'(Pose=Component To Local(ComponentPose=@n40)) {cacheName=skydive_post_sim}
   Use cached pose 'copypose_cache'()
   Use cached pose 'copypose_cache'()
   Use cached pose 'copypose_cache'()
  ]}
 EventGraph {nodes [
   Event BlueprintUpdateAnimation()
   > n1 = Sequence!(execute) -> (then_0, then_1, then_2:Exec, then_3:Exec, then_4:Exec, then_5:Exec, then_6:Exec, then_7:Exec)
   n2 = Get bIsInFrontEnd()
   > Set Is not In Front End(Is not In Front End:double="To Float (Boolean)"(InBool=NOT Boolean(A=@n2))) -> (Output_Get:double)
   n3 = Branch(execute=@n1.then_1, Condition=Get bClothEnabled())
   > n4 = Cloth Controller Function DEPRECATED!(
    execute
    CharacterPartAnimInstance:Custom Character Part Anim Instance Object Reference=Self-Reference()
    ClothingAssetName:String=M_MED_BlackMonday_Houston_Pack_Clothing_2
    Cloth Paramaters:Cloth Params Struct Structure=Get Cloth Paramaters_0() -> (Cloth Paramaters_0:Cloth Params Struct Structure)
    Cloth Core Settings:Cloth Core Setting Struct Structure=Get Cloth Core Settings_0() -> (Cloth Core Settings_0:Cloth Core Setting Struct Structure)
    ClothUpAxis:Vector="0,0,1"
    OnGround_Moving_SpeedLimit:double=2
    OnGround_Moving_AngularVel_Max:double=10
    OnGround_Moving_FictitiousVel_Max:double=0.6
    LinVelScale_FwdBwd_Clamp:double=0
    LinVelScale_UpDwn_Clamp:double=0.7
    DeltaTime:double=0
    GameTimeInSeconds:double=GetGameTimeInSeconds()
    WindGustsParams:Wind Gust Params Struct Structure
    WindGustLocalSettings:Wind Gust Local Setting Struct Structure
   ) -> (
    then
    CurrentStateOut:String
    Cloth Settings Out:Cloth Setting Struct Structure
    Cloth Core Settings Out:Cloth Core Setting Struct Structure
    WindGustSettings Out:Wind Gust Setting Struct Structure
    WindGustLocalSettings Out:Wind Gust Local Setting Struct Structure
    GustStrength - WindVelocityOut:Vector
    LinearVelVector-DELETE LATER:Vector
   )
   > n5 = Cloth Controller Function DEPRECATED(
    CharacterPartAnimInstance=Self-Reference()
    ClothingAssetName=M_MED_BlackMonday_Kansas_Pack_Clothing_0
    Cloth Paramaters=Get Cloth Paramaters_0()
    Cloth Core Settings=Get Cloth Core Settings_0()
    ClothUpAxis="0,0,1"
    OnGround_Moving_SpeedLimit=2
    OnGround_Moving_AngularVel_Max=10
    OnGround_Moving_FictitiousVel_Max=0.6
    LinVelScale_FwdBwd_Clamp=0
    LinVelScale_UpDwn_Clamp=0.7
    DeltaTime=0
    GameTimeInSeconds=GetGameTimeInSeconds()
   )
   n6 = Gravity Override State Function DEPRECATED(
    execute=@n3.else
    CharacterPartAnimInstance=Self-Reference()
    GameTimeInSeconds=GetGameTimeInSeconds()
    GravityOverrideParams=Get Gravity Override Params()
   )
   > Set GravityOverride(GravityOverride=@n6)
   n7 = Branch(execute=@n1.then_2, Condition=Get bIsSkyDiving())
   > Set SkydivingAlpha(SkydivingAlpha:Rotator="0,-30,0") -> (Output_Get:Rotator)
   Set SkydivingAlpha(execute=@n7.else, SkydivingAlpha="0,-10,0")
   n8 = PlayerMovement(execute=@n1.then_3, AnimInstance=Self-Reference(), XDotProductCoefficient=1, YDotProductCoefficient=1, ZDotProductCoefficient=1)
   > n9 = Branch(Condition=AND Boolean(A=Get bIsTargeting(), B=Get bIsCrouching()))
   n10 = MapRangeClamped(
    Value=float + float(A="Absolute (Float)"(A=@n8), B=MapRangeClamped(Value=@n8.YMovement, InRangeA=-1, InRangeB=1, OutRangeA=0.3)) {is=title K2Node_CommutativeAssociativeBinaryOperator}
    InRangeA=0
    InRangeB=1.65
   )
   > Set BwdMovement(BwdMovement:double=float * float(A=@n10, B=0.25) {is=title K2Node_PromotableOperator}) -> (Output_Get:double)
   Set BwdMovement(execute=@n9.else, BwdMovement=@n10)
   n11 = Branch(execute=@n1.then_4, Condition=Get bIsSkyDiving())
   > n12 = Branch(Condition="float > float"(A=Get PawnVelocityZ(), B=50) {is=title K2Node_CallFunction})
   > Set BwdMovement(BwdMovement=float * float(A=Get BwdMovement(), B=0) {is=title K2Node_CommutativeAssociativeBinaryOperator})
   Set BwdMovement(execute=[@n12.else, @n11.else], BwdMovement=float * float(A=Get BwdMovement(), B=1))
   n13 = GetPartSkeletalMeshComponent()
   Set TeleportDistanceThreshold!(execute=@n1.then_5, TeleportDistanceThreshold:float=300, self:Skeletal Mesh Component Object Reference=@n13) -> (then, Output_Get:float)
   > Set TeleportRotationThreshold!(execute, TeleportRotationThreshold:float=15, self:Skeletal Mesh Component Object Reference=@n13) -> (then, Output_Get:float)
   > n14 = Set Current LOD(Current LOD:int=Get CurrentLOD()) -> (Output_Get:int=0)
   > Set lowLODEnabled(lowLODEnabled:bool="integer >= integer"(A=@n14, B=2)) -> (Output_Get:bool)
   n15 = Get capeOn() -> (capeOn:bool)
   > Set animDynamicLoinclothOn(animDynamicLoinclothOn:bool=OR Boolean(A=Get bClothEnabled(), B=NOT Boolean(A=@n15))) -> (Output_Get:bool)
   > Set isBackblingInFrontend(isBackblingInFrontend:bool=AND Boolean(A=Get bIsInFrontEnd(), B=Get Is Wearing Backbling())) -> (Output_Get:bool)
   n16 = NOT Boolean(A=Get bRigidBodyEnabled())
   n17 = NOT Boolean(A=Get bClothEnabled())
   > Set trailControllerConditions(trailControllerConditions:bool=OR Boolean(A=Get lowLODEnabled() -> (lowLODEnabled:bool), B=AND Boolean(A=@n17, B=@n16))) -> (Output_Get:bool)
   > n18 = Branch(Condition=OR Boolean(A=Get bIsParachuteOpened(), B=Get bIsSkyDiving()))
   n19 = BreakTransform(InTransform=GetDeltaTransformFromRefPose(self=GetPartSkeletalMeshComponent(), BoneName=upperarm_l, BaseName=spine_05))
   > Set upperarm_Alpha(upperarm_Alpha:double=MapRangeClamped(Value=BreakRotator(InRot=@n19.Rotation).Yaw, InRangeA=-5, InRangeB=-30)) -> (Output_Get:double)
   Set upperarm_Alpha(execute=@n18.else, upperarm_Alpha=0)
   n20 = GetClothingSimulationInteractor!(self:Skeletal Mesh Component Object Reference=GetOwningComponent(), ClothingAssetName:Name=None) -> (ReturnValue:Clothing Simulation Interactor Object Reference)
   Is Valid(exec=@n1.then_7, InputObject=@n20)
   n21 = MakeLiteralString(Value=M_MED_BlackMonday_Kansas_Pack_Clothing_0)
   > n22 = Cast To ChaosClothingInteractor!(
    execute
    Object=GetClothingInteractor!(self:Clothing Simulation Interactor Object Reference=@n20, ClothingAssetName:Name=String To Name(InString=@n21)) -> (ReturnValue:Clothing Interactor Object Reference)
   ) -> (then, CastFailed, AsChaos Clothing Interactor:Chaos Clothing Interactor Object Reference) {is=title K2Node_DynamicCast}
   > Set ChaosClothingInteractor(ChaosClothingInteractor:Chaos Clothing Interactor Object Reference=@n22) -> (Output_Get:Chaos Clothing Interactor Object Reference)
   n23 = GetClothingSimulationInteractor(self=GetOwningComponent(), ClothingAssetName=None)
   Is Valid(exec=@n1.then_6, InputObject=@n23)
   n24 = MakeLiteralString(Value=M_MED_BlackMonday_Houston_Pack_Clothing_2)
   > n25 = Cast To ChaosClothingInteractor(Object=GetClothingInteractor(self=@n23, ClothingAssetName=String To Name(InString=@n24)))
   > Set ChaosClothingInteractor(ChaosClothingInteractor=@n25)
   Event OnCharacterCustomizationLoaded()
   > n26 = Sequence!(execute) -> (then_0, then_1)
   > Set Is Wearing Backbling(Is Wearing Backbling=CheckForBackbling())
   n27 = Branch(execute=@n26.then_1, Condition=CharacterItemDefinitionHasAnyMetaTag(Tags=Make Array("[0]"=(TagName="Cosmetics.Character.RestrictCape"))))
   > Set capeOn(capeOn:bool=true) -> (Output_Get:bool)
   Set capeOn(execute=@n27.else, capeOn=false)
   n28 = Event BlueprintProcessAnimInputEvent()
   > n29 = Branch(Condition=IsNamed(self=@n28.AnimInputEvent, InEventName=aie.Player.Zipline.Entered))
   > Delay(Duration=0.1)
   n30 = Get ChaosClothingInteractor() -> (ChaosClothingInteractor:Chaos Clothing Interactor Object Reference)
   > ResetAndTeleport!(execute, self:Chaos Clothing Interactor Object Reference=@n30, bReset:bool=true, bTeleport:bool=true) -> (then)
   Branch(execute=@n29.else, Condition=IsNamed(self=@n28.AnimInputEvent, InEventName=aie.Player.Zipline.Exited))
   > Set resetClothLandingFromZilpine(resetClothLandingFromZilpine:bool=true) -> (Output_Get:bool)
   Event OnEndFalling()
   > Branch(Condition=Get resetClothLandingFromZilpine() -> (resetClothLandingFromZilpine:bool))
   > ResetAndTeleport(self=@n30, bReset=true, bTeleport=true)
   > Set resetClothLandingFromZilpine(resetClothLandingFromZilpine=false)
   Event BlueprintInitializeAnimation()
   > Set bResetAfterTeleport!(execute, bResetAfterTeleport:bool=false, self:Skeletal Mesh Component Object Reference=@n13) -> (then, Output_Get:bool)
   Break Cloth Setting Struct!(ClothSettingStruct:"Cloth Setting Struct Structure (by ref)"=@n4.Cloth Settings Out) -> (
    LinearVelocityScale_7_#1:Vector=~
    AngularVelocityScale_6_#2:float=~
    FictitiousAngularScale_10_#3:float=~
    GravityOverride_15_#4:Vector=~
    DragCoefficient_23_#5:Vector 2D Structure
    LiftCoefficient_24_#6:Vector 2D Structure
    JointName_27_#7:Name
   ) {is=title K2Node_BreakStruct}
   Break Cloth Setting Struct(ClothSettingStruct=@n5.Cloth Settings Out) -> (
    LinearVelocityScale_7_#1=~
    AngularVelocityScale_6_#2=~
    FictitiousAngularScale_10_#3=~
    GravityOverride_15_#4=~
   )
  ]}
]
variables [
 animDynamicLoinclothOn {type=boolean}
 BwdMovement {type=number}
 capeOn {type=boolean}
 ChaosClothingInteractor {type=ChaosClothingInteractor, default=null}
 Cloth Core Settings_0 {
  type=ClothCoreSettingStruct
  default {
   collisionThickness=1.5
   dampingCoefficient=0.1
   frictionCoefficient=0.5
   localDampingCoefficient=0.1
   numberIterations=2
   numberSubsteps=1
   selfCollisionThickness=0
   animDriveDamping=<0 1>
   animDriveStiffness=<0 1>
   areaStiffness=<1 1>
   bendingStiffness=<1 1>
   edgeStiffness=<1 1>
   tetherStiffness=<1 1>
  }
 }
 Cloth Paramaters_0 {
  type=ClothParamsStruct
  default {
   dBNO {
    angularVelocityScale=0.6
    fictitiousAngularScale=0.1
    jointName=pelvis
    dragCoefficient=<0.2 1>
    gravityOverride=<-980 400 0>
    liftCoefficient=<0 0.5>
    linearVelocityScale=<0.3 0.3 0.1>
   }
   dBNO_Carried {
    angularVelocityScale=0.6
    fictitiousAngularScale=0.1
    jointName=pelvis
    dragCoefficient=<0.2 1>
    gravityOverride=<-980 -200 0>
    liftCoefficient=<0 0.5>
    linearVelocityScale=<0.3 0.3 0.1>
   }
   dBNO_Impulse {
    angularVelocityScale=0
    fictitiousAngularScale=0
    jointName=pelvis
    dragCoefficient=<0 0.2>
    gravityOverride=<-3000 -2000 0>
    liftCoefficient=<0 0.1>
    linearVelocityScale=<0.3 0.1 0.3>
   }
   dBNO_RollBwd {
    angularVelocityScale=0
    fictitiousAngularScale=0
    jointName=pelvis
    dragCoefficient=<0 0.2>
    gravityOverride=<-2000 0 0>
    liftCoefficient=<0 0.1>
    linearVelocityScale=<0.3 0.1 0.3>
   }
   dBNO_RollFwd {
    angularVelocityScale=0
    fictitiousAngularScale=0
    jointName=pelvis
    dragCoefficient=<0 0.2>
    gravityOverride=<-980 0 0>
    liftCoefficient=<0 0.1>
    linearVelocityScale=<0.3 0.1 0.3>
   }
   dBNO_Tumbling {
    angularVelocityScale=0
    fictitiousAngularScale=0
    jointName=pelvis
    dragCoefficient=<0 0.2>
    gravityOverride=<-2500 -300 0>
    liftCoefficient=<0 0.1>
    linearVelocityScale=<0.3 0.1 0.3>
   }
   emoteOrMelee {
    angularVelocityScale=0.6
    fictitiousAngularScale=0.1
    jointName=root
    dragCoefficient=<2.5 2.5>
    gravityOverride=<0 0 -980>
    liftCoefficient=<0.5 0.5>
    linearVelocityScale=<0.6 0.6 0.5>
   }
   front_End {
    angularVelocityScale=0.05
    fictitiousAngularScale=0.6
    jointName=root
    dragCoefficient=<0.2 0.2>
    gravityOverride=<0 -200 -980>
    liftCoefficient=<0.1 0.1>
    linearVelocityScale=<0.1 0.1 0.1>
   }
   onGround_Crouch_Moving {
    angularVelocityScale=0.1
    fictitiousAngularScale=0.3
    jointName=root
    dragCoefficient=<2 2>
    gravityOverride=<0 0 -980>
    liftCoefficient=<0.5 0.5>
    linearVelocityScale=<0.6 0.6 0.5>
   }
   onGround_Falling {
    angularVelocityScale=0.6
    fictitiousAngularScale=0
    jointName=root
    dragCoefficient=<0 0.2>
    gravityOverride=<0 -500 -980>
    liftCoefficient=<0 0.1>
    linearVelocityScale=<2 1 0.3>
   }
   onGround_Moving {
    angularVelocityScale=0.6
    fictitiousAngularScale=0.1
    jointName=root
    dragCoefficient=<2 2>
    gravityOverride=<0 0 -980>
    liftCoefficient=<0.5 0.5>
    linearVelocityScale=<0.6 0.6 0.5>
   }
   onGround_Standing {
    angularVelocityScale=0.05
    fictitiousAngularScale=1.2
    jointName=root
    dragCoefficient=<2.5 2.5>
    gravityOverride=<0 0 -980>
    liftCoefficient=<0.5 0.5>
    linearVelocityScale=<0.6 0.6 0.5>
   }
   rail_Grinding {
    angularVelocityScale=0.2
    fictitiousAngularScale=0
    jointName=root
    dragCoefficient=<0.1 0.2>
    gravityOverride=<50 -500 -400>
    liftCoefficient=<0 0.5>
    linearVelocityScale=<0.1 0.1 0.1>
   }
   riding_Moving {
    angularVelocityScale=0.05
    fictitiousAngularScale=1
    jointName=root
    dragCoefficient=<0.5 0.5>
    gravityOverride=<0 0 -980>
    liftCoefficient=<0.5 0.5>
    linearVelocityScale=<0.4 0.6 0.3>
   }
   riding_Moving_Targeting {
    angularVelocityScale=0.05
    fictitiousAngularScale=0.7
    jointName=root
    dragCoefficient=<0.5 0.5>
    gravityOverride=<0 0 -980>
    liftCoefficient=<0.5 0.5>
    linearVelocityScale=<0.2 0.6 0.2>
   }
   riding_Standing {
    angularVelocityScale=0.05
    fictitiousAngularScale=1
    jointName=root
    dragCoefficient=<0.5 0.5>
    gravityOverride=<0 -450 -980>
    liftCoefficient=<0.5 0.5>
    linearVelocityScale=<0.6 0.6 0.3>
   }
   skydive_DivingDown {
    angularVelocityScale=0.1
    fictitiousAngularScale=0
    jointName=pelvis
    dragCoefficient=<0.3 0.3>
    gravityOverride=<-450 -100 0>
    liftCoefficient=<0.9 0.9>
    linearVelocityScale=<0.1 0.1 0.1>
   }
   skydive_FloatingDown {
    angularVelocityScale=0.1
    fictitiousAngularScale=0
    jointName=root
    dragCoefficient=<0.3 0.3>
    gravityOverride=<0 -800 200>
    liftCoefficient=<1.3 1.3>
    linearVelocityScale=<0.2 0.2 0.1>
   }
   skydive_GoingUp {
    angularVelocityScale=0.1
    fictitiousAngularScale=0
    jointName=root
    dragCoefficient=<0.01 0.01>
    gravityOverride=<0 -2000 400>
    liftCoefficient=<0.05 0.05>
    linearVelocityScale=<0.1 0.1 0.01>
   }
   skydive_Parachuting {
    angularVelocityScale=0.6
    fictitiousAngularScale=0.1
    jointName=root
    dragCoefficient=<0.8 0.8>
    gravityOverride=<0 -600 -500>
    liftCoefficient=<0.8 0.8>
    linearVelocityScale=<0.2 0.2 0.3>
   }
   skydive_Surfing {
    angularVelocityScale=0.6
    fictitiousAngularScale=0.1
    jointName=root
    dragCoefficient=<1.5 1.5>
    gravityOverride=<0 -600 -500>
    liftCoefficient=<1 1>
    linearVelocityScale=<0.1 0.1 0.3>
   }
   swimming {
    angularVelocityScale=0.6
    fictitiousAngularScale=0
    jointName=root
    dragCoefficient=<2.5 2.5>
    gravityOverride=<0 0 0>
    liftCoefficient=<0.1 0.1>
    linearVelocityScale=<0.6 0.6 0.5>
   }
   tactical_Sprint {
    angularVelocityScale=0.6
    fictitiousAngularScale=0.1
    jointName=root
    dragCoefficient=<2 2>
    gravityOverride=<0 0 -980>
    liftCoefficient=<0.5 0.5>
    linearVelocityScale=<0.6 0.6 0.5>
   }
   ziplining {
    angularVelocityScale=0.6
    fictitiousAngularScale=0
    jointName=root
    dragCoefficient=<0 1.5>
    gravityOverride=<0 0 -980>
    liftCoefficient=<0 0.3>
    linearVelocityScale=<0.4 0.4 0.5>
   }
  }
 }
 Current LOD {type=integer}
 Gravity Override Params {type=GravityOverrideParamsStruct}
 GravityOverride {type=Vector}
 Is not In Front End {type=number}
 Is Wearing Backbling {type=boolean}
 isBackblingInFrontend {type=boolean}
 lowLODEnabled {type=boolean}
 resetClothLandingFromZilpine {type=boolean}
 SkydivingAlpha {type=Rotator}
 trailControllerConditions {type=boolean}
 upperarm_Alpha {type=number}
]
